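# Translucent, unlit, two-sided particle material for the engine exhaust effect on the Ashton Salt Lake glider.
# Panning noise layers and a V-gradient fire line form flames, hue-shifted along V; a Soft Glow switch substitutes a noise-distorted radial ramp.
# Dynamic parameters drive line contrast, UV offset, hue contrast and a vertex-color-masked flare offset.
# Opacity combines Fresnel fade and depth fade (skipped on mobile), with flames adding vertex-color and edge-texture masks.
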Material /BRCosmetics/Gadgets/Assets/VinderTech_GliderChute/Glider_Ashton_Salt_Lake/FX/M_Ashton_SaltLake_Engines
graph {
 properties {
  bUsedWithMeshParticles=true
  bUsedWithNiagaraMeshParticles=true
  blendMode=BLEND_Translucent
  pixelDepthOffsetMode=PDOM_Legacy
  shadingModel=MSM_Unlit
  twoSided=true
  usedShadingModels=Unlit
 }
 nodes [
  n1 = TextureCoordinate()
  n2 = Panner(Coordinate=TextureCoordinate() {uTiling=1.4}) {speedY=-0.2}
  n3 = Multiply(
   A=ComponentMask(None=@n1) {g=true}
   B=ComponentMask(None=TextureSample(UVs=@n2) {texture=&T_NoisePack_01}.RGB) {g=true, r=true}
  )
  n4 = TextureSample(UVs=Add(A=TextureCoordinate(), B=Multiply(A=@n3) {constB=0.14})) {texture=/Game/Effects/Fort_Effects/Textures/Ramps/T_Ramp_Radial_Soft}
  n5 = ParticleColor()
  n6 = OneMinus(None=ComponentMask(None=TextureCoordinate()) {g=true})
  n7 = LinearInterpolate(
   A=ScalarParameter() {defaultValue=2, parameterName=FireLineContrast A}
   B=ScalarParameter() {defaultValue=5, parameterName=FireLineContrast B}
   Alpha=DynamicParameter() {paramNames [FireLineControl, UVOffset, HueContrast, WPO_Flare]}.FireLineControl
  ) {constA=2, constB=5}
  n8 = ComponentMask(None=VectorParameter() {group=UVScale, parameterName=UVScale, defaultValue=<rgba 1 0.25 0 0>}.RGB) {g=true, r=true}
  n9 = DynamicParameter() {paramNames [FireLineControl, UVOffset, HueContrast, WPO_Flare]}
  n10 = Add(A=Multiply(A=TextureCoordinate() {vTiling=0.5}, B=@n8), B=@n9.UVOffset)
  n11 = MaterialFunctionCall(
   In=TextureSample(UVs=Panner(Coordinate=@n10) {speedX=-0.05, speedY=-1}) {texture=&T_NoisePack_01}.G
   Contrast=Constant() {r=0.2}
  ) {materialFunction=&CheapContrast}
  n12 = Multiply(A=@n8, B=TextureCoordinate() {uTiling=4})
  n13 = Multiply(A=Time(), B=ScalarParameter() {defaultValue=5, parameterName=Speed})
  n14 = TextureSample(UVs=Panner(Coordinate=Add(A=@n12, B=@n9.UVOffset), Time=@n13) {speedX=0.1, speedY=-1}) {texture=&T_NoisePack_01}
  n15 = TextureCoordinate() {uTiling=6, vTiling=1.5}
  n16 = Panner(Coordinate=Add(A=@n9.UVOffset, B=Multiply(A=@n8, B=@n15)), Time=@n13) {speedX=-0.1, speedY=-1}
  n17 = Add(
   A=Multiply(A=Power(Base=@n6, Exp=@n7), B=ScalarParameter() {defaultValue=0.6, parameterName=FireLineIntensity})
   B=Multiply(A=Add(A=@n11.Result, B=@n14.G), B=TextureSample(UVs=@n16) {texture=&T_NoisePack_01}.G)
  )
  n18 = Desaturation(
   None=StaticSwitchParameter(True=Multiply(A=@n4.R, B=@n5.RGB), False=Multiply(A=@n17, B=@n5.RGB)) {parameterName=Soft Glow}
   Fraction=ScalarParameter() {defaultValue=-1, parameterName=DesatFactor}
  )
  n19 = ComponentMask(None=TextureCoordinate()) {g=true}
  n20 = LinearInterpolate(
   A=MaterialFunctionCall(Hue Shift Percentage=ScalarParameter() {defaultValue=0.4381, parameterName=Hue Shift Amount}, Texture=@n18) {materialFunction=&HueShift}.Result
   B=@n18
   Alpha=Power(Base=OneMinus(None=@n19), Exp=DynamicParameter() {paramNames [FireLineControl, UVOffset, HueContrast, WPO_Flare]}.HueContrast)
  )
  MaterialFunctionCall(Emissive=MaterialFunctionCall(LDR Input=Clamp(None=@n20), HDR Tint=@n20) {materialFunction=&DeriveHDRfromLDR}.RemappedRange) {materialFunction=&MF_FX_EmissiveAdjustments}
  Multiply()
  n21 = VertexNormalWS()
  n22 = Transform(None=VectorParameter() {parameterName=Local Vector, defaultValue=<rgba 0 1 0 0>}.RGB) {transformSourceType=TRANSFORMSOURCE_Local}
  n23 = ScalarParameter() {parameterName=Flare WPO by Vector}
  Multiply(
   A=OneMinus(None=VertexColor().R)
   B=Multiply(
    A=LinearInterpolate(A=@n21, B=@n22, Alpha=@n23)
    B=DynamicParameter() {paramNames [FireLineControl, UVOffset, HueContrast, WPO_Flare]}.WPO_Flare
   )
  )
  n24 = Fresnel(ExponentIn=ScalarParameter() {defaultValue=0.2, group=Opacity, parameterName=FresExp, sortPriority=4}) {exponent=0.2}
  n25 = Clamp(None=Multiply(A=OneMinus(None=@n24)) {constB=10})
  n26 = ComponentMask(None=@n1) {r=true}
  n27 = Abs(None=@n26)
  n28 = OneMinus(None=@n26)
  n29 = Multiply(A=@n25, B=Multiply(A=@n4.R, B=Multiply(A=@n27, B=@n28)))
  n30 = DepthFade(Opacity=@n29, FadeDistance=ScalarParameter() {defaultValue=50, group=Opacity, parameterName=Depth Fade Distance, sortPriority=1})
  n31 = ScalarParameter() {defaultValue=1, group=Opacity, parameterName=Vertex Color Opacity, sortPriority=3}
  n32 = ChannelMaskParameter(None=VertexColor()) {group=Opacity, maskChannel=Green, parameterName=Vertex Color, sortPriority=2, defaultValue=<rgba 0 1 0 0>}
  n33 = OneMinus(None=@n32.R)
  n34 = MaterialFunctionCall(In=Power(Base=@n17) {constExponent=0.5}, Contrast=ScalarParameter() {defaultValue=0.5, parameterName=HardEdgeness}) {materialFunction=&CheapContrast}
  n35 = TextureCoordinate()
  n36 = Multiply(A=Multiply(A=@n5.A, B=Clamp(None=@n34.Result)), B=OneMinus(None=ComponentMask(None=@n35) {g=true}))
  n37 = MaterialFunctionCall(Float2=TextureCoordinate()) {materialFunction=&BreakOutFloat2Components}
  n38 = OneMinus(None=@n37.G)
  n39 = ScalarParameter() {defaultValue=1, group=Alpha Edge, parameterName=Alpha Edge Invert V, sortPriority=4}
  n40 = Multiply(
   A=AppendVector(A=@n37.R, B=LinearInterpolate(A=@n37.G, B=@n38, Alpha=@n39))
   B=ComponentMask(None=VectorParameter() {group=Alpha Edge, parameterName=Alpha Edge UV Scale, sortPriority=2, defaultValue=<rgba 1 1 0 0>}.RGB) {g=true, r=true}
  )
  n41 = VectorParameter() {group=Alpha Edge, parameterName=Alpha Edge UV Offset, sortPriority=3, defaultValue=<rgba 0 -0.2 0 0>}
  n42 = ComponentMask(None=@n41.RGB) {g=true, r=true}
  n43 = TextureSampleParameter2D(UVs=Panner(Coordinate=Add(A=@n40, B=@n42)) {speedX=1}) {
   group=Alpha Edge
   parameterName=Alpha Edge
   samplerType=SAMPLERTYPE_LinearColor
   texture=/Game/Effects/Fort_Effects/Textures/Flares/T_Smash_Flare_Packed
  }
  n44 = ChannelMaskParameter(None=@n43.RGBA) {group=Alpha Edge, maskChannel=Blue, parameterName=Alpha Edge Mask, sortPriority=1, defaultValue=<rgba 0 0 1 0>}
  n45 = Multiply(A=Multiply(A=LinearInterpolate(A=@n31, Alpha=@n33) {constA=1}, B=Multiply(A=@n36, B=@n25)), B=@n44.R)
  n46 = DepthFade(Opacity=@n45, FadeDistance=ScalarParameter() {defaultValue=50, group=Opacity, parameterName=Depth Fade Distance, sortPriority=1})
  StaticSwitchParameter(
   True=Multiply(A=ParticleColor().A, B=ShadingPathSwitch(Default=@n30, Mobile=@n29) {desc=Migrated from FeatureLevelSwitch})
   False=LinearInterpolate(
    A=@n45
    B=ShadingPathSwitch(Default=@n46, Mobile=@n45) {desc=Migrated from FeatureLevelSwitch}
    Alpha=ScalarParameter() {group=Opacity, parameterName=Use Depth Fade}
   )
  ) {parameterName=Soft Glow}
 ]
}
