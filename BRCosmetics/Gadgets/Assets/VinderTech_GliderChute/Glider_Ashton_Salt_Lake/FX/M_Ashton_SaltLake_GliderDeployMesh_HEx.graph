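# Translucent unlit Niagara mesh-particle material for the Ashton Salt Lake glider deploy effect, drawing glowing blue energy over circuit patterns.
# Radially mapped, panning noise layers blend two HDR colors for the emissive look.
# A gradient swept by dynamic parameter LightningControl (looping time under the DEBUG switch) reveals two scrolling circuitboard layers.
# Red vertex color adds extra circuit intensity, and particle alpha scales the combined mask.
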Material /BRCosmetics/Gadgets/Assets/VinderTech_GliderChute/Glider_Ashton_Salt_Lake/FX/M_Ashton_SaltLake_GliderDeployMesh_HEx
graph {
 properties {
  bUsedWithNiagaraMeshParticles=true
  blendMode=BLEND_Translucent
  pixelDepthOffsetMode=PDOM_Legacy
  shadingModel=MSM_Unlit
  usedShadingModels=Unlit
 }
 nodes [
  n1 = TextureCoordinate() {coordinateIndex=1}
  n2 = ScalarParameter() {defaultValue=1, parameterName=UScale_Fire}
  n3 = ScalarParameter() {defaultValue=1, parameterName=VScale_Fire}
  n4 = AppendVector(A=@n2, B=@n3)
  n5 = Multiply(A=@n2) {constB=0.5}
  n6 = Multiply(A=@n3) {constB=0.5}
  n7 = MaterialFunctionCall(Vector or UVs=Subtract(A=Multiply(A=@n1, B=@n4), B=AppendVector(A=@n5, B=@n6))) {materialFunction=&VectorToRadialValue}
  n8 = Panner(Coordinate=@n7.Radial Coordinates) {speedX=-0.5, speedY=-1}
  n9 = Multiply(A=Time(), B=ScalarParameter() {defaultValue=-0.5, group=FlameShape, parameterName=Flame_Speed})
  n10 = Add(
   A=MaterialFunctionCall(In=TextureSample(UVs=@n8) {texture=&T_NoisePack_01}.G, Contrast=Constant() {r=0.2}) {materialFunction=&CheapContrast}.Result
   B=TextureSample(UVs=Panner(Coordinate=@n7.Radial Coordinates, Time=@n9) {speedX=0.2, speedY=1}) {texture=&T_NoisePack_01}.G
  )
  n11 = TextureSample(UVs=Panner(Coordinate=@n7.Radial Coordinates, Time=@n9) {speedX=-0.2, speedY=1}) {samplerType=SAMPLERTYPE_LinearColor, texture=/Game/Characters/Player/Male/Medium/Bodies/M_MED_SRDrift/Textures/T_ElectricNoise}
  n12 = OneMinus(None=ComponentMask(None=@n7.Radial Coordinates) {g=true})
  n13 = ScalarParameter() {defaultValue=20}
  n14 = Power(Base=Add(A=Multiply(A=@n10, B=@n11.G), B=Power(Base=@n12, Exp=@n13))) {constExponent=0.5}
  n15 = ScalarParameter() {defaultValue=0.5, parameterName=ContrastCheap}
  n16 = LinearInterpolate(
   A=VectorParameter() {group=FlameShape, parameterName=Color_A, defaultValue=<rgba 0.8416 4.505 10 1>}.RGB
   B=VectorParameter() {group=FlameShape, parameterName=Color_B, defaultValue=<rgba 0.9946 6.9982 10 1>}.RGB
   Alpha=Clamp(None=MaterialFunctionCall(In=@n14, Contrast=@n15) {materialFunction=&CheapContrast}.Result)
  )
  MaterialFunctionCall(Emissive=MaterialFunctionCall(LDR Input=Clamp(None=@n16), HDR Tint=@n16) {materialFunction=&DeriveHDRfromLDR}.RemappedRange) {materialFunction=&MF_FX_EmissiveAdjustments}
  n17 = DynamicParameter() {parameterIndex=1, defaultValue=<rgba 1 1 0.15 1>, paramNames [LightningControl, Param2, Param3, Param4]}
  n18 = StaticSwitchParameter(True=Time() {bOverride_Period=true, period=1}, False=@n17.LightningControl) {parameterName=DEBUG}
  n19 = ScalarParameter() {defaultValue=-0.72, parameterName=TextureTimeOffset}
  n20 = TextureSample(UVs=Panner(Coordinate=@n7.Radial Coordinates, Time=Add(A=@n18, B=@n19)) {speedY=-1}) {texture=/Game/Global/Textures/Default/Masks_and_Gradients/gradient_stasis}
  n21 = TextureCoordinate()
  n22 = VectorParameter() {parameterName=Circuit_Tiling, defaultValue=<rgba 0.25 2 0 0>}
  n23 = ComponentMask(None=VectorParameter() {parameterName=Circuit_SpeedValue, defaultValue=<rgba 0 1.52 0 0>}.RGB) {g=true, r=true}
  n24 = Panner(
   Coordinate=Multiply(
    A=MaterialFunctionCall(XY=@n21) {materialFunction=&Swizzle}.YX
    B=ComponentMask(None=@n22.RGB) {g=true, r=true}
   )
   Speed=@n23
  ) {speedY=1}
  n25 = MaterialFunctionCall(XY=TextureCoordinate()) {materialFunction=&Swizzle}
  n26 = ComponentMask(None=VectorParameter() {parameterName=CircuitOffset, defaultValue=<rgba -0.2095 0.3048 0 0>}.RGB) {g=true, r=true}
  n27 = VectorParameter() {parameterName=Circuit_Tiling_1, defaultValue=<rgba 0.25 -2 0 0>}
  n28 = Panner(Coordinate=Multiply(A=Add(A=@n25.YX, B=@n26), B=ComponentMask(None=@n27.RGB) {g=true, r=true}), Speed=@n23)
  n29 = Clamp(None=Add(
    A=TextureSample(UVs=@n24) {texture=/Game/Effects/Fort_Effects/Textures/T_FX_Circuitboard}.G
    B=TextureSample(UVs=@n28) {texture=&T_FX_Circuitboard}.R
   ))
  n30 = Multiply(A=VertexColor().R, B=@n29)
  n31 = ScalarParameter() {defaultValue=0.5, parameterName=VertRedIntensity}
  Multiply(A=ParticleColor().A, B=Add(A=Multiply(A=@n20.R, B=@n29), B=Multiply(A=@n30, B=@n31)))
  ScalarParameter() {defaultValue=0.15, parameterName=Period, sliderMax=1}
  n32 = Constant()
  n33 = Subtract(A=TextureCoordinate()) {constB=0.5}
  n34 = AppendVector(A=@n32, B=ComponentMask(None=Multiply(A=@n33) {constB=2}) {g=true})
  Transform(None=Multiply(A=AppendVector(A=@n34, B=@n32), B=ScalarParameter() {defaultValue=5, parameterName=Base}))
 ]
}
