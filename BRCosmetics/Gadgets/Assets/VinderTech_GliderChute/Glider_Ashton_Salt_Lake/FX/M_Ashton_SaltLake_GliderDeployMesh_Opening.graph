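# Unlit translucent Niagara mesh-particle material for the opening deploy effect of the Ashton Salt Lake glider.
# A V-panning noise texture scales a tint multiplied by particle color, while a separately panning star mask drives HDR intensity for bright sparkles.
# Alpha combines the noise, an inverted Fresnel and contrast-boosted top and bottom UV edge fades, then multiplies by particle alpha.
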Material /BRCosmetics/Gadgets/Assets/VinderTech_GliderChute/Glider_Ashton_Salt_Lake/FX/M_Ashton_SaltLake_GliderDeployMesh_Opening
graph {
 properties {
  bUsedWithNiagaraMeshParticles=true
  blendMode=BLEND_Translucent
  pixelDepthOffsetMode=PDOM_Legacy
  shadingModel=MSM_Unlit
  usedShadingModels=Unlit
 }
 nodes [
  n1 = Multiply(
   A=TextureCoordinate()
   B=AppendVector(A=ScalarParameter() {defaultValue=4, parameterName=UScale}, B=ScalarParameter() {defaultValue=0.5, parameterName=VScale})
  )
  n2 = LinearInterpolate(Alpha=TextureSample(UVs=Panner(Coordinate=@n1) {bFractionalPart=true, speedY=1}) {texture=&2ndInfectedMapLinear}.G) {constA=0.25}
  n3 = Multiply(
   A=LinearInterpolate(B=VectorParameter() {parameterName=LDR_Color, defaultValue=<rgba 0.0378 0.0935 0.1771 1>}.RGB, Alpha=@n2)
   B=ParticleColor().RGB
  )
  n4 = ScalarParameter() {defaultValue=1000, parameterName=Star_HDRIntensity}
  n5 = TextureCoordinate()
  n6 = AppendVector(A=ScalarParameter() {defaultValue=5, parameterName=Star_UScale}, B=ScalarParameter() {defaultValue=1, parameterName=Star_VScale})
  n7 = TextureSample(UVs=Panner(Coordinate=Multiply(A=@n5, B=@n6)) {bFractionalPart=true, speedY=1}) {samplerType=SAMPLERTYPE_Masks, texture=/Game/Packages/Fortress_Sky/Textures/T_Sky_Stars_Sharp}
  MaterialFunctionCall(Emissive=MaterialFunctionCall(LDR Input=Clamp(None=@n3), HDR Tint=@n3, HDR Intensity Multiplier=LinearInterpolate(B=@n4, Alpha=@n7.R) {constA=1, constB=100}) {materialFunction=&DeriveHDRfromLDR}.RemappedRange) {materialFunction=&MF_FX_EmissiveAdjustments}
  n8 = Fresnel(ExponentIn=ScalarParameter() {defaultValue=0.5, parameterName=Alpha_FresExp})
  n9 = ComponentMask(None=TextureCoordinate()) {g=true}
  n10 = Multiply(
   A=Power(Base=@n9, Exp=ScalarParameter() {defaultValue=2, parameterName=TopEdge})
   B=Power(Base=OneMinus(None=@n9), Exp=ScalarParameter() {defaultValue=1, parameterName=BottomEdge})
  )
  n11 = ScalarParameter() {defaultValue=10, parameterName=Contrast Compensation}
  n12 = Multiply(A=Multiply(A=@n2, B=OneMinus(None=@n8)), B=Clamp(None=Multiply(A=@n10, B=@n11)))
  Multiply(A=@n12)
  Multiply(A=ParticleColor().A, B=@n12)
 ]
}
