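# Translucent unlit Niagara mesh-particle material for the middle effect of the Ashton Salt Lake glider deploy.
# Object-relative triplanar projections build bump-offset triangle shards and wireframe lines, modulated by radial panning noise and UV sphere masks.
# Rotating zoom noise, distorted by blended hex and triangle normals, joins them to drive an HDR two-color emissive.
# Dynamic parameters control normal distortion and opacity contrast; particle alpha scales opacity, and masked inverse Fresnel drives refraction.
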Material /BRCosmetics/Gadgets/Assets/VinderTech_GliderChute/Glider_Ashton_Salt_Lake/FX/M_Ashton_SaltLake_GliderDeploy_Middle
graph {
 properties {
  bUsedWithNiagaraMeshParticles=true
  blendMode=BLEND_Translucent
  pixelDepthOffsetMode=PDOM_Legacy
  refractionMethod=RM_IndexOfRefraction
  shadingModel=MSM_Unlit
  usedShadingModels=Unlit
 }
 nodes [
  n1 = TextureCoordinate() {coordinateIndex=1}
  n2 = Constant() {r=0.5}
  n3 = ScalarParameter() {defaultValue=0.525, parameterName=OpacityMasterRadius}
  n4 = ScalarParameter() {parameterName=OpacityMasterHardness}
  n5 = Fresnel()
  LinearInterpolate(
   B=ScalarParameter() {defaultValue=1.025, parameterName=Refraction Index}
   Alpha=Multiply(A=SphereMask(A=@n1, B=@n2, Radius=@n3, Hardness=@n4), B=OneMinus(None=@n5))
  ) {constA=1, constB=1.1}
  n6 = VectorParameter() {parameterName=DetailNormal_Size, defaultValue=<rgba 100 100 100 0>}
  n7 = Divide(
   A=Subtract(A=WorldPosition() {worldPositionShaderOffset=WPT_ExcludeAllShaderOffsets}.XYZ, B=ObjectPositionWS())
   B=Multiply(A=Abs(None=@n6.RGB)) {constB=-1}
  )
  n8 = ComponentMask(None=@n7) {b=true, r=true}
  n9 = Constant2Vector() {g=0.125}
  n10 = ScalarParameter() {defaultValue=-1, parameterName=Tris_BumpHeight}
  n11 = TextureSample(UVs=BumpOffset(Coordinate=Add(A=@n8, B=@n9), Height=@n10)) {texture=/Game/Gadgets/Assets/VinderTech_GliderChute/Glider_Ashton_Salt_Lake/FX/T_AshtonSaltLake_JumpTri}
  n12 = ComponentMask(None=@n7) {b=true, g=true}
  n13 = TextureSample(UVs=BumpOffset(Coordinate=Add(A=@n12, B=@n9), Height=@n10)) {texture=&T_AshtonSaltLake_JumpTri}
  n14 = Transform(None=VertexNormalWS()) {transformSourceType=TRANSFORMSOURCE_World, transformType=TRANSFORM_Local}
  n15 = ComponentMask(None=@n14) {r=true}
  n16 = ScalarParameter() {defaultValue=1, parameterName=Proj_Contrast}
  n17 = ComponentMask(None=MaterialFunctionCall(In=Abs(None=@n15), Contrast=@n16) {materialFunction=&CheapContrast}.Result) {r=true}
  n18 = BumpOffset(Coordinate=Add(A=ComponentMask(None=@n7) {g=true, r=true}, B=@n9), Height=@n10)
  n19 = MaterialFunctionCall(In=Abs(None=ComponentMask(None=@n14) {b=true}), Contrast=@n16) {materialFunction=&CheapContrast}
  n20 = TextureCoordinate() {coordinateIndex=1}
  n21 = AppendVector(A=ScalarParameter() {defaultValue=0.85, parameterName=U_Sphere}, B=ScalarParameter() {defaultValue=1, parameterName=V_Sphere})
  n22 = Constant() {r=0.5}
  n23 = Multiply(
   A=LinearInterpolate(
    A=LinearInterpolate(A=@n11.R, B=@n13.R, Alpha=@n17) {desc=Float3}
    B=TextureSample(UVs=@n18) {texture=&T_AshtonSaltLake_JumpTri}.R
    Alpha=ComponentMask(None=@n19.Result) {r=true}
   )
   B=SphereMask(A=Multiply(A=@n20, B=@n21), B=Multiply(A=@n22, B=@n21), Radius=ScalarParameter() {defaultValue=0.3, parameterName=Shard_Radius}) {hardnessPercent=0}
  )
  n24 = ScalarParameter() {defaultValue=2, parameterName=UScale_Fire}
  n25 = ScalarParameter() {defaultValue=2, parameterName=VScale_Fire}
  n26 = Subtract(
   A=Multiply(A=TextureCoordinate() {coordinateIndex=1}, B=AppendVector(A=@n24, B=@n25))
   B=AppendVector(A=Multiply(A=@n24) {constB=0.5}, B=Multiply(A=@n25) {constB=0.5})
  )
  n27 = TextureSample(UVs=Panner(Coordinate=MaterialFunctionCall(Vector or UVs=@n26) {materialFunction=&VectorToRadialValue}.Radial Coordinates) {speedX=0.125, speedY=-1}) {texture=&T_NoisePack_01}
  n28 = LinearInterpolate(Alpha=@n27.G) {constA=0.25, constB=1.25}
  n29 = VectorParameter() {parameterName=Texture Size, defaultValue=<rgba 50 50 25 0>}
  n30 = Divide(
   A=Subtract(A=WorldPosition() {worldPositionShaderOffset=WPT_ExcludeAllShaderOffsets}.XYZ, B=ObjectPositionWS())
   B=Multiply(A=Abs(None=@n29.RGB)) {constB=-1}
  )
  n31 = TextureSample(UVs=ComponentMask(None=@n30) {b=true, r=true}) {texture=&/T_Wireframe}
  n32 = ScalarParameter() {defaultValue=10, parameterName=RedIntensity}
  n33 = ScalarParameter() {defaultValue=2, parameterName=BlueContrast}
  n34 = TextureSample(UVs=ComponentMask(None=@n30) {b=true, g=true}) {texture=&T_Wireframe}
  n35 = Transform(None=VertexNormalWS()) {transformSourceType=TRANSFORMSOURCE_World, transformType=TRANSFORM_Local}
  n36 = Abs(None=ComponentMask(None=@n35) {r=true})
  n37 = ScalarParameter() {defaultValue=1, parameterName=Proj_Contrast}
  n38 = LinearInterpolate(
   A=Add(A=Multiply(A=@n31.R, B=@n32), B=Power(Base=@n31.B, Exp=@n33))
   B=Add(A=Multiply(A=@n34.R, B=@n32), B=Power(Base=@n34.B, Exp=@n33))
   Alpha=ComponentMask(None=MaterialFunctionCall(In=@n36, Contrast=@n37) {materialFunction=&CheapContrast}.Result) {r=true}
  ) {desc=Float3}
  n39 = TextureSample(UVs=ComponentMask(None=@n30) {g=true, r=true}) {texture=&T_Wireframe}
  n40 = Add(A=Multiply(A=@n39.R, B=@n32), B=Power(Base=@n39.B, Exp=@n33))
  n41 = ComponentMask(None=@n35) {b=true}
  n42 = ComponentMask(None=MaterialFunctionCall(In=Abs(None=@n41), Contrast=@n37) {materialFunction=&CheapContrast}.Result) {r=true}
  n43 = SphereMask(A=TextureCoordinate() {coordinateIndex=1}, B=Constant() {r=0.5}, Radius=ScalarParameter() {defaultValue=0.75, parameterName=Lines_Radius}) {hardnessPercent=0}
  n44 = Multiply(A=@n43, B=@n43)
  n45 = TextureSample(UVs=Rotator(Coordinate=TextureCoordinate() {coordinateIndex=1})) {texture=/Game/Effects/Fort_Effects/BluGlow/ZoomNoise}
  n46 = Multiply(A=LinearInterpolate(Alpha=@n27.R) {constA=0.05}, B=Multiply(A=LinearInterpolate(A=@n38, B=@n40, Alpha=@n42), B=Multiply(A=@n44, B=@n45.R)))
  n47 = ScalarParameter() {defaultValue=2, parameterName=Lines Intensity}
  n48 = TextureSample() {
   samplerType=SAMPLERTYPE_Normal
   texture=&/T_Ashton_SaltLake_Jump_N
  }
  n49 = MaterialFunctionCall(Normal=@n48.RGB, Flatness=ScalarParameter() {defaultValue=0.6, parameterName=HexNormFlatness}) {materialFunction=&FlattenNormal}
  n50 = VectorParameter() {parameterName=DetailNormal_Size, defaultValue=<rgba 100 100 100 0>}
  n51 = Divide(
   A=Subtract(A=WorldPosition() {worldPositionShaderOffset=WPT_ExcludeAllShaderOffsets}.XYZ, B=ObjectPositionWS())
   B=Multiply(A=Abs(None=@n50.RGB)) {constB=-1}
  )
  n52 = ComponentMask(None=@n51) {b=true, r=true}
  n53 = Constant2Vector() {g=0.125}
  n54 = ScalarParameter() {defaultValue=-1, parameterName=Tris_BumpHeight}
  n55 = TextureSample(UVs=BumpOffset(Coordinate=Add(A=@n52, B=@n53), Height=@n54)) {
   samplerType=SAMPLERTYPE_Normal
   texture=&/T_AshtonSaltLake_JumpTri_N
  }
  n56 = ComponentMask(None=@n51) {b=true, g=true}
  n57 = TextureSample(UVs=BumpOffset(Coordinate=Add(A=@n56, B=@n53), Height=@n54)) {
   samplerType=SAMPLERTYPE_Normal
   texture=&T_AshtonSaltLake_JumpTri_N
  }
  n58 = Transform(None=VertexNormalWS()) {transformSourceType=TRANSFORMSOURCE_World, transformType=TRANSFORM_Local}
  n59 = ComponentMask(None=@n58) {r=true}
  n60 = ScalarParameter() {defaultValue=1, parameterName=Proj_Contrast}
  n61 = ComponentMask(None=MaterialFunctionCall(In=Abs(None=@n59), Contrast=@n60) {materialFunction=&CheapContrast}.Result) {r=true}
  n62 = ComponentMask(None=@n51) {g=true, r=true}
  n63 = TextureSample(UVs=BumpOffset(Coordinate=Add(A=@n62, B=@n53), Height=@n54)) {
   samplerType=SAMPLERTYPE_Normal
   texture=&T_AshtonSaltLake_JumpTri_N
  }
  n64 = MaterialFunctionCall(In=Abs(None=ComponentMask(None=@n58) {b=true}), Contrast=@n60) {materialFunction=&CheapContrast}
  n65 = MaterialFunctionCall(
   BaseNormal=@n49.Result
   AdditionalNormal=LinearInterpolate(A=LinearInterpolate(A=@n55.RGB, B=@n57.RGB, Alpha=@n61) {desc=Float3}, B=@n63.RGB, Alpha=ComponentMask(None=@n64.Result) {r=true})
  ) {materialFunction=&BlendAngleCorrectedNormals}
  n66 = TextureCoordinate() {coordinateIndex=1}
  n67 = Constant() {r=0.5}
  n68 = ScalarParameter() {defaultValue=0.5, parameterName=HexNormal_Radius}
  n69 = ScalarParameter() {defaultValue=0.3381, parameterName=HexNormal_Hardness}
  n70 = Clamp(None=Power(
    Base=SphereMask(A=@n66, B=@n67, Radius=@n68, Hardness=@n69) {hardnessPercent=0}
    Exp=ScalarParameter() {defaultValue=1.4155, parameterName=NormalBlendContrast}
   ) {constExponent=0.5})
  n71 = MaterialFunctionCall(
   Normal=LinearInterpolate(
    A=Constant3Vector() {constant=<rgba 0 0 1 0>}
    B=LinearInterpolate(A=@n65.Result, B=@n49.Result, Alpha=@n70)
    Alpha=DynamicParameter() {paramNames [Opacity,Normal,Param3,Param4]}.Normal
   )
   Flatness=ScalarParameter() {defaultValue=0.6857}
  ) {materialFunction=&FlattenNormal}
  n72 = Rotator(Coordinate=Add(A=TextureCoordinate() {coordinateIndex=1}, B=ComponentMask(None=@n71.Result) {g=true, r=true})) {speed=-0.25}
  n73 = SphereMask(A=TextureCoordinate() {coordinateIndex=1}, B=Constant() {r=0.5}, Radius=ScalarParameter() {defaultValue=0.75, parameterName=Fresnel_Radius}) {hardnessPercent=0}
  n74 = Add(
   A=Add(A=Multiply(A=@n23, B=@n28), B=Multiply(A=@n46, B=@n47))
   B=Multiply(A=TextureSample(UVs=@n72) {texture=&ZoomNoise}.R, B=Multiply(A=@n73, B=@n73))
  )
  n75 = LinearInterpolate(
   A=VectorParameter() {group=FlameShape, parameterName=Color_A, defaultValue=<rgba 0.1683 0.901 2 1>}.RGB
   B=VectorParameter() {group=FlameShape, parameterName=Color_B, defaultValue=<rgba 0.5 0.8333 1 1>}.RGB
   Alpha=Clamp(None=@n74)
  )
  MaterialFunctionCall(Emissive=MaterialFunctionCall(LDR Input=Clamp(None=@n75), HDR Tint=@n75, HDR Intensity Multiplier=ScalarParameter() {defaultValue=1, parameterName=HDR_INtensity}) {materialFunction=&DeriveHDRfromLDR}.RemappedRange) {materialFunction=&MF_FX_EmissiveAdjustments}
  n76 = DynamicParameter() {paramNames [Opacity,Normal,Param3,Param4]}
  Multiply(
   A=ParticleColor().A
   B=MaterialFunctionCall(In=@n74, Contrast=LinearInterpolate(Alpha=@n76.Opacity) {constA=1, constB=0.025}) {materialFunction=&CheapContrast}.Result
  )
  ScalarParameter() {defaultValue=1, parameterName=NormalDrive}
 ]
}
