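# Translucent unlit Niagara mesh-particle material that gives the Ashton Salt Lake glider deploy effect an outer-space look.
# Emissive blends camera-vector triplanar nebula and portal-space textures by a star mask, then tints and remaps them to HDR.
# A separate mask combines radially panned noise with a sphere mask and a view-angle fade.
# The noise sample uses explicit pre-polar UV derivatives, except on the mobile shading path.
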Material /BRCosmetics/Gadgets/Assets/VinderTech_GliderChute/Glider_Ashton_Salt_Lake/FX/M_Ashton_SaltLake_GliderDeploy_Space
graph {
 properties {
  bUsedWithNiagaraMeshParticles=true
  blendMode=BLEND_Translucent
  ditherOpacityMask=true
  pixelDepthOffsetMode=PDOM_Legacy
  shadingModel=MSM_Unlit
  usedShadingModels=Unlit
 }
 nodes [
  n1 = TextureObject() {texture=/Game/Characters/Player/Male/Medium/Bodies/M_Med_Soldier_04/Skins/Galaxy/Textures/T_FN_Nebula}
  n2 = ScalarParameter() {defaultValue=0.5, parameterName=Nebula Tile}
  n3 = VectorParameter() {parameterName=Offset_A, defaultValue=<rgba 0 0.16 0 0>}
  n4 = TextureObject() {texture=/Game/Effects/Fort_Effects/Textures/Tile/T_Portal_Space_D}
  n5 = ScalarParameter() {defaultValue=1, parameterName=Portal Tile}
  n6 = TextureSample() {samplerType=SAMPLERTYPE_Masks, texture=/Game/Characters/Player/Male/Medium/Bodies/M_Med_Soldier_04/Skins/Galaxy/Textures/Celestial_T_Stars}
  n7 = Multiply(
   A=LinearInterpolate(
    A=MaterialFunctionCall!(CameraVector, Texture=@n1, Tiling=@n2, Offset=@n3.RGB, AxisFadeContrast) {materialFunction=/Engine/Functions/Engine_MaterialFunctions01/Texturing/TriplanarCameraVector}.Result
    B=MaterialFunctionCall(Texture=@n4, Tiling=@n5) {materialFunction=&TriplanarCameraVector}.Result
    Alpha=@n6.G
   )
   B=VectorParameter() {parameterName=HDR_Tint, defaultValue=<rgba 0.5 0.5937 1 1>}.RGB
  )
  MaterialFunctionCall(Emissive=MaterialFunctionCall(LDR Input=Clamp(None=@n7), HDR Tint=@n7, HDR Intensity Multiplier=ScalarParameter() {defaultValue=5, parameterName=HDR_Tint}) {materialFunction=&DeriveHDRfromLDR}.RemappedRange) {materialFunction=&MF_FX_EmissiveAdjustments}
  n8 = ComponentMask(None=VectorParameter() {defaultValue=<rgba 0.5 0.5 0 0>}.RGB) {g=true, r=true}
  n9 = Subtract(A=Multiply(A=TextureCoordinate(), B=@n8), B=Multiply(A=@n8) {constB=0.5})
  n10 = Panner(Coordinate=MaterialFunctionCall(Vector or UVs=@n9) {materialFunction=&VectorToRadialValue}.Radial Coordinates) {speedY=-0.7}
  n11 = ShadingPathSwitch(
   Default=TextureSample!(UVs=@n10, Tex, "DDX(UVs)"=DDX(Value=@n9), "DDY(UVs)"=DDY(Value=@n9), Apply View MipBias) {mipValueMode=TMVM_Derivative, texture=&T_NoisePack_01}.G
   Mobile=TextureSample!(UVs=@n10, Tex, Apply View MipBias) {texture=&T_NoisePack_01}.G
  )
  n12 = SphereMask(
   A=TextureCoordinate()
   B=Constant() {r=0.5}
   Radius=ScalarParameter() {defaultValue=0.5, parameterName=Shard_Radius}
   Hardness=ScalarParameter() {parameterName=HArdness}
  )
  Multiply(
   A=Multiply(A=Add(A=@n11, B=@n12), B=LinearInterpolate(Alpha=@n12) {constB=2})
   B=MaterialFunctionCall("FadeSensitivity (optional)"=ScalarParameter() {defaultValue=1.5, parameterName=FadeAngle Sen}) {materialFunction=&FadeBasedOnViewAngle}.Result
  )
 ]
}
